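# Masked, skinned material for the Treasury deployable turret, textured with Treasury maps and the Moonflax emissive, clipped below a local Z height.
# World position offset combines Animation-driven arm rotation and sliding, timestamp-driven recoil, and a hit-direction wobble disabled on ES2/ES3.1.
# Normals follow the rotation, and attributes pass through MF_CreativeMoveTool.
# The tinted fresnel emissive, time pulse and height-mask branches are unconnected, so emissive is the raw texture.
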Material /Game/Athena/Items/EnvironmentalItems/Sentry/Turret/Treasury/Materials/M_DeployableTurretTreasury
graph {
 properties {
  bUseMaterialAttributes=true
  bUsedWithMeshDeformer=true
  bUsedWithSkeletalMesh=true
  blendMode=BLEND_Masked
  pixelDepthOffsetMode=PDOM_Legacy
  usedShadingModels=Default Lit
 }
 nodes [
  n1 = Time()
  n2 = ScalarParameter() {defaultValue=0.5, parameterName=Speed}
  MaterialFunctionCall!(In=Fmod(A=Multiply(A=@n1, B=@n2), B=Constant() {r=1})) {materialFunction=/Game/Effects/Fort_Effects/Materials/_Functions/MF_Smoothstep_01}
  n3 = TextureSample() {texture=/Game/Athena/Items/EnvironmentalItems/Sentry/Turret/Moonflax/Textures/T_DeployableTurret_Moonflax_E}
  n4 = Multiply(
   A=StaticSwitchParameter(
    True=VectorParameter() {bUseCustomPrimitiveData=true, parameterName=EmissiveColor_PrimitiveData, defaultValue=<rgba 0 0 0 1>}.RGB
    False=VectorParameter() {parameterName=EmissiveColor, defaultValue=<rgba 0 0 0 1>}.RGB
   ) {parameterName=UsePrimitiveData}
   B=StaticSwitchParameter(
    True=ScalarParameter() {bUseCustomPrimitiveData=true, defaultValue=1, parameterName=EmissiveScale_PrimitiveData, primitiveDataIndex=4}
    False=ScalarParameter() {defaultValue=1, parameterName=EmissiveScale}
   ) {parameterName=UsePrimitiveData}
  )
  n5 = ScalarParameter() {defaultValue=0.5, parameterName=FresnelEmissiveLow}
  n6 = MaterialFunctionCall() {materialFunction=&MF_Fresnel_FX}
  MaterialFunctionCall(Emissive=Multiply(A=Multiply(A=@n3.RGB, B=@n4), B=LinearInterpolate(A=@n5, Alpha=@n6.Fresnel))) {materialFunction=&MF_FX_EmissiveAdjustments}
  n7 = TextureSample() {samplerType=SAMPLERTYPE_Masks, texture=/Game/Athena/Items/EnvironmentalItems/Sentry/Turret/Treasury/Textures/T_DeployableTurret_Treasury_S}
  n8 = ScalarParameter() {defaultValue=-50, parameterName=ZHeight}
  n9 = ComponentMask(None=TransformPosition(None=WorldPosition().XYZ) {transformSourceType=TRANSFORMPOSSOURCE_World}) {b=true}
  n10 = Constant()
  n11 = Constant() {r=1}
  n12 = Normalize(VectorInput=Transform(None=Constant3Vector() {constant=<rgba 0 1 0 0>}) {transformSourceType=TRANSFORMSOURCE_Local})
  n13 = ScalarParameter() {defaultValue=0.8, parameterName=AngleEnd}
  n14 = ScalarParameter() {defaultValue=0.4, parameterName=AngleStart}
  n15 = StaticSwitchParameter(
   True=ScalarParameter() {bUseCustomPrimitiveData=true, defaultValue=1, parameterName=Animation_PrimitiveData, primitiveDataIndex=6}
   False=ScalarParameter() {defaultValue=1, parameterName=Animation}
  ) {parameterName=UsePrimitiveData}
  n16 = Subtract(A=Time() {period=1}, B=ScalarParameter() {parameterName=Recoil})
  n17 = ScalarParameter() {defaultValue=0.25, parameterName=RecoilDuration}
  n18 = MaterialFunctionCall!(In=Saturate(None=Divide(A=@n16, B=@n17))) {materialFunction=/Game/Effects/Fort_Effects/Materials/_Functions/MF_Smoothstep_EaseOut}
  n19 = Add(
   A=Multiply(
    A=MaterialFunctionCall!(min=@n13, max=@n14, value=@n15) {materialFunction=/Game/Effects/Fort_Effects/Materials/_Functions/InverseLerp_MF}.Clamped
    B=ScalarParameter() {defaultValue=-0.25, parameterName=ArmRotationAngle}
   )
   B=Multiply(A=ScalarParameter() {defaultValue=-0.007, parameterName=RecoilPitch}, B=@n18.Result)
  )
  n20 = TextureSample() {samplerType=SAMPLERTYPE_Normal, texture=/Game/Athena/Items/EnvironmentalItems/Sentry/Turret/Treasury/Textures/T_DeployableTurret_Treasury_N}
  n21 = Transform(None=@n20.RGB)
  n22 = Transform(None=VectorParameter() {parameterName=MoveOffset, defaultValue=<rgba 0 0 -100 0>}.RGB) {transformSourceType=TRANSFORMSOURCE_Local}
  n23 = MaterialFunctionCall(min=ScalarParameter() {defaultValue=0.8, parameterName=MoveOffsetEnd}, max=ScalarParameter() {parameterName=MoveOffsetStart}, value=@n15) {materialFunction=&InverseLerp_MF}
  n24 = TransformPosition(None=Constant3Vector()) {transformType=TRANSFORMPOSSOURCE_World}
  n25 = WorldPosition()
  n26 = Power(Base=@n18.Result, Exp=ScalarParameter() {defaultValue=0.6, parameterName=RecoilBackwardPower}) {constExponent=0.3}
  n27 = Transform(None=Constant3Vector() {constant=<rgba 1 0 0 0>}) {transformSourceType=TRANSFORMSOURCE_Local}
  n28 = ScalarParameter() {defaultValue=-2, parameterName=RecoilBackward}
  n29 = Add(
   A=Add(A=Multiply(A=@n22, B=@n23.Clamped), B=RotateAboutAxis(NormalizedRotationAxis=@n12, RotationAngle=@n19, PivotPoint=@n24, Position=@n25.XYZ))
   B=Multiply(A=Saturate(None=@n26), B=Multiply(A=@n27, B=@n28))
  )
  n30 = TransformPosition(None=Constant3Vector()) {transformType=TRANSFORMPOSSOURCE_World}
  n31 = Subtract(A=VectorParameter() {parameterName=HitLocation, defaultValue=<rgba 0 50 50 0>}.RGB, B=@n30)
  n32 = Normalize(VectorInput=CrossProduct(A=Normalize(VectorInput=@n31), B=Constant3Vector() {constant=<rgba 0 0 1 0>}))
  n33 = Subtract(A=Time(), B=ScalarParameter() {group=z_blueprint, parameterName=Wobble})
  n34 = ScalarParameter() {defaultValue=0.34, group=WorldPositionOffset, parameterName=WobbleDuration}
  n35 = MaterialFunctionCall(In=Saturate(None=Divide(A=@n33, B=@n34))) {materialFunction=&MF_Smoothstep_EaseOut}
  n36 = Multiply(A=ScalarParameter() {defaultValue=0.025, group=WorldPositionOffset, parameterName=WobbleAmount}, B=Sine(None=@n35.Result) {period=1.5})
  n37 = WorldPosition()
  n38 = Constant()
  n39 = FeatureLevelSwitch(
   Default=Multiply(A=RotateAboutAxis(NormalizedRotationAxis=@n32, RotationAngle=@n36, PivotPoint=@n30, Position=@n37.XYZ), B=@n35.Result)
   ES2=@n38
   ES3_1=@n38
  )
  MaterialFunctionCall(MA In=MakeMaterialAttributes!(
    BaseColor=TextureSample() {texture=/Game/Athena/Items/EnvironmentalItems/Sentry/Turret/Treasury/Textures/T_DeployableTurret_Treasury_D}.RGB
    Metallic=@n7.G
    Specular=@n7.G
    Roughness=@n7.G
    Anisotropy
    EmissiveColor=TextureSample() {texture=&T_DeployableTurret_Moonflax_E}.RGB
    Opacity
    OpacityMask=If(A=@n8, B=@n9, "A > B"=@n10, "A == B"=@n10, "A < B"=@n11)
    Normal=MaterialFunctionCall(Rotation Axis=@n12, Rotation Angle=@n19, World Space Vertex Normals=@n21) {materialFunction=&FixRotateAboutAxisNormals}.New Tangent Space Vertex Normal
    Tangent
    WorldPositionOffset=Add(A=@n29, B=@n39)
    SubsurfaceColor
    ClearCoat
    ClearCoatRoughness
    AmbientOcclusion
    Refraction
    CustomizedUVs_0
    CustomizedUVs_1
    CustomizedUVs_2
    CustomizedUVs_3
    CustomizedUVs_4
    CustomizedUVs_5
    CustomizedUVs_6
    CustomizedUVs_7
    PixelDepthOffset
    ShadingModel
    Displacement
   )) {materialFunction=&MF_CreativeMoveTool}
  n40 = WorldPosition()
  n41 = ObjectPositionWS()
  n42 = ObjectRadius()
  n43 = ComponentMask(None=Add(A=Subtract(A=@n40.XYZ, B=@n41), B=@n42)) {b=true}
  Saturate(None=Add(A=Divide(A=@n43, B=@n42), B=ScalarParameter() {defaultValue=0.8, group=WorldPositionOffset, parameterName=WPO_HeightMaskAtten, sliderMax=1}))
  Transform(None=Constant3Vector() {constant=<rgba 0 1 0 0>}) {transformSourceType=TRANSFORMSOURCE_Local}
 ]
}
