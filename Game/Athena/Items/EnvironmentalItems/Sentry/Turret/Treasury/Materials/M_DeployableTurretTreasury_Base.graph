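# Masked, lit base material for the Treasury deployable turret, with diffuse, normal and mask textures, where the mask's green channel drives metallic, specular and roughness.
# Emissive combines a turret texture, color, scale and Fresnel falloff, then passes through MF_FX_EmissiveAdjustments.
# A Blades value cuts red-vertex-color areas along UV channel 2 via a sphere mask.
# UsePrimitiveData switches emissive and Blades inputs to custom primitive data; MF_CreativeMoveTool wraps the result.
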Material /Game/Athena/Items/EnvironmentalItems/Sentry/Turret/Treasury/Materials/M_DeployableTurretTreasury_Base
graph {
 properties {bUseMaterialAttributes=true, blendMode=BLEND_Masked, pixelDepthOffsetMode=PDOM_Legacy, usedShadingModels=Default Lit}
 nodes [
  n1 = TextureSample() {
   samplerType=SAMPLERTYPE_Masks
   texture=/Game/Athena/Items/EnvironmentalItems/Sentry/Turret/Treasury/Textures/T_DeployableTurret_Treasury_Base_S
  }
  n2 = StaticSwitchParameter(
   True=VectorParameter() {bUseCustomPrimitiveData=true, parameterName=EmissiveColor_PrimitiveData, defaultValue=<rgba 0 0 0 1>}.RGB
   False=VectorParameter() {parameterName=EmissiveColor, defaultValue=<rgba 0 0 0 1>}.RGB
  ) {parameterName=UsePrimitiveData}
  n3 = StaticSwitchParameter(
   True=ScalarParameter() {bUseCustomPrimitiveData=true, defaultValue=1, parameterName=EmissiveScale_PrimitiveData, primitiveDataIndex=4}
   False=ScalarParameter() {defaultValue=1, parameterName=EmissiveScale}
  ) {parameterName=UsePrimitiveData}
  n4 = Multiply(
   A=Multiply(A=TextureSample() {texture=/Game/Creative/Sets/PropSets/Playgrounds/Textures/T_Turret_Base_E}.RGB, B=Multiply(A=@n2, B=@n3))
   B=LinearInterpolate(
    A=ScalarParameter() {defaultValue=0.5, parameterName=FresnelEmissiveLow}
    Alpha=MaterialFunctionCall() {materialFunction=&MF_Fresnel_FX}.Fresnel
   )
  )
  n5 = TextureCoordinate() {coordinateIndex=2}
  n6 = SphereMask(
   A=OneMinus(None=ComponentMask(None=@n5) {g=true})
   B=Constant()
   Radius=StaticSwitchParameter(
    True=ScalarParameter() {bUseCustomPrimitiveData=true, defaultValue=1, parameterName=Blades_PrimitiveData, primitiveDataIndex=5}
    False=ScalarParameter() {defaultValue=1, parameterName=Blades}
   ) {parameterName=UsePrimitiveData}
   Hardness=Constant() {r=1}
  )
  n7 = VertexColor()
  n8 = TextureSample() {
   samplerType=SAMPLERTYPE_Normal
   texture=/Game/Athena/Items/EnvironmentalItems/Sentry/Turret/Treasury/Textures/T_DeployableTurret_Treasury_Base_N
  }
  MaterialFunctionCall(MA In=MakeMaterialAttributes!(
    BaseColor=TextureSample() {texture=/Game/Athena/Items/EnvironmentalItems/Sentry/Turret/Treasury/Textures/T_DeployableTurret_Treasury_Base_D}.RGB
    Metallic=@n1.G
    Specular=@n1.G
    Roughness=@n1.G
    Anisotropy
    EmissiveColor=MaterialFunctionCall(Emissive=@n4) {materialFunction=&MF_FX_EmissiveAdjustments}.Result
    Opacity
    OpacityMask=LinearInterpolate(B=@n6, Alpha=@n7.R) {constA=1}
    Normal=@n8.RGB
    Tangent
    WorldPositionOffset
    SubsurfaceColor
    ClearCoat
    ClearCoatRoughness
    AmbientOcclusion
    Refraction
    CustomizedUVs_0
    CustomizedUVs_1
    CustomizedUVs_2
    CustomizedUVs_3
    CustomizedUVs_4
    CustomizedUVs_5
    CustomizedUVs_6
    CustomizedUVs_7
    PixelDepthOffset
    ShadingModel
    Displacement
   )) {materialFunction=&MF_CreativeMoveTool}
 ]
}
